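# UI translucent material for a mosaic tile in the Season 28 Battle Pass Odyssey quest screen.
# It multiplies the alpha of a MosaicTileTexture parameter by a clamped Discovered scalar, which defaults to one.
# The Discovered parameter thus scales how visible the tile's alpha shape is, presumably to reveal tiles as they are found.
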Material /BattlePassS28_Quests/UI/Odyssey/M_S28_Odyssey_Mosaic
graph {
 properties {blendMode=BLEND_Translucent, materialDomain=MD_UI, pixelDepthOffsetMode=PDOM_Legacy, usedShadingModels=Default Lit}
 nodes [Multiply(
   A=TextureSampleParameter2D() {parameterName=MosaicTileTexture, texture=/Game/Athena/UI/InGame/Sidekick/Reactions/100}.A
   B=Clamp(None=ScalarParameter() {defaultValue=1, parameterName=Discovered})
  )]
}
